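; EQS context for creature AI that provides a single location: the querier's last known position, as the LKP name indicates.
; It casts the querier actor to FortPawn, finds its FortPawnComponent_AlertLevel and returns that component's last known position.
; If the cast fails or the component is not valid, it falls back to the querier actor's own location.
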(fn ProvideSingleLocation
  (params (QuerierObject :type Object%20Reference) (QuerierActor :type Actor%20Object%20Reference))
  (bind _asfort_pawn (Utilities|Casting|CastToFortPawn QuerierActor)
    (:then
      (Utilities|IsValid (Actor|GetComponentbyClass _asfort_pawn "/Script/FortniteAI.FortPawnComponent_AlertLevel")
        (:Is%20Valid
          (return (AlertLevelComponent|GetLastKnownPosition (Actor|GetComponentbyClass _asfort_pawn "/Script/FortniteAI.FortPawnComponent_AlertLevel"))))
        (:Is%20Not%20Valid
          _)))
    (:CastFailed
      _))
  (return (Transformation|GetActorLocation (Variables|GetQuerierActor))))
