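; EQS context blueprint that supplies a single location for the querier, used by creature AI environment queries.
; It casts the querier actor to Pawn and returns its nav agent location.
; If the querier is not a Pawn, it falls back to the plain actor location.
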(fn ProvideSingleLocation
  (params (QuerierObject :type Object%20Reference) (QuerierActor :type Actor%20Object%20Reference))
  (bind _aspawn (Utilities|Casting|CastToPawn QuerierActor)
    (:then
      (return (Pawn|GetNavAgentLocation _aspawn)))
    (:CastFailed
      (return (Transformation|GetActorLocation QuerierActor)))))
